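; Looping GameplayCue for the Heat Vision targeted beam, handling its eye beams, impact effects and audio.
; On start it finds the targeted beam component on the pawn or weapon, wires Niagara systems to tick last, and attaches beams to head or weapon eye sockets.
; The weapon's beam audio actor supplies sound, with a repeating visualizer indicator for non-locally viewed pawns.
; Removal stops audio, hides beams and shows ending effects, including impact only if one occurred.
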(event EventOnLoopingStart(EitherVFX-preferthisversion)
  (params (MyTarget :type Actor%20Object%20Reference) (Parameters :type Gameplay%20Cue%20Parameters%20Structure%20%28by%20ref%29) (ParticleComponents :type Array%20of%20FXSystem%20Component%20Object%20References) (AudioComponents :type Array%20of%20Audio%20Component%20Object%20References))
  (bind _returnvalue_bisactive (Utilities|Struct|BreakTargetedBeamData (Beam|GetBeamData (Variables|FortGameplayCueNotifyLoop_TargetedBeam|GetTargetedBeamActorComponent))))
  (bind _beamcomponents (Variables|FortGameplayCueNotifyLoop_TargetedBeam|GetBeamComponents))
  (bind _output (Utilities|Array|Get(acopy) _beamcomponents 2))
  (bind _particlecomponents (Variables|FortGameplayCueNotifyLoop_TargetedBeam|GetParticleComponents))
  (bind _output_2 (Utilities|Array|Get(acopy) _particlecomponents 1))
  (for _array_element ParticleComponents
    (bind _asniagara_particle_system_component (Utilities|Casting|CastToNiagaraComponent _array_element))
    (Niagara|SetTickBehavior _asniagara_particle_system_component "ForceTickLast"))
  (Variables|FortGameplayCueNotifyLoop_TargetedBeam|SetbBeamStarted false)
  (bind _output_get (Variables|FortGameplayCueNotifyLoop_TargetedBeam|SetParticleComponents ParticleComponents))
  (Rendering|SetVisibility _output_get false true)
  (Utilities|FlowControl|Sequence
    (:then_0
      (Variables|FortGameplayCueNotifyLoop_TargetedBeam|SetbIsLocallyControlled (Ability|GameplayCue|IsInstigatorLocallyControlled Parameters))
      (CallFunction|InitializeBeamAudio MyTarget true)
      (CallFunction|LocateTargetedBeamComponent MyTarget)
      (Actor|Tick|AddTickPrerequisiteComponent (Variables|FortGameplayCueNotifyLoop_TargetedBeam|GetTargetedBeamActorComponent))
      (Variables|FortGameplayCueNotifyLoop_TargetedBeam|SetBeamComponents (Utilities|Array|MakeArray (Utilities|Array|Get(acopy) (Variables|FortGameplayCueNotifyLoop_TargetedBeam|GetParticleComponents) 4) (Utilities|Array|Get(acopy) (Variables|FortGameplayCueNotifyLoop_TargetedBeam|GetParticleComponents) 5) (Utilities|Array|Get(acopy) (Variables|FortGameplayCueNotifyLoop_TargetedBeam|GetParticleComponents) 6) (Utilities|Array|Get(acopy) (Variables|FortGameplayCueNotifyLoop_TargetedBeam|GetParticleComponents) 7)))
      (Variables|FortGameplayCueNotifyLoop_TargetedBeam|SetImpactFXComponent _output_2)
      (Variables|FortGameplayCueNotifyLoop_TargetedBeam|SetImpactFXComponent_Light (Utilities|Array|Get(acopy) _particlecomponents 3))
      (Rendering|SetTranslucentSortPriority _output_2 1)
      (Rendering|SetTranslucentSortPriority (Utilities|Array|Get(acopy) _particlecomponents 2) 2)
      (Variables|FortGameplayCueNotifyLoop_TargetedBeam|SetWaterFXComponent (Utilities|Array|Get(acopy) (Variables|FortGameplayCueNotifyLoop_TargetedBeam|GetParticleComponents) 8))
      (if _returnvalue_bisactive
        (CallFunction|StartBeam)))
    (:then_1
      ; BeamAttachment
      (CallFunction|AttachSystemtoEyePositions MyTarget (Utilities|Array|MakeArray (Utilities|Array|Get(acopy) _beamcomponents) _output) (Utilities|Array|MakeArray (Utilities|Array|Get(acopy) _beamcomponents 1) (Utilities|Array|Get(acopy) _beamcomponents 3)) (Variables|Default|GetbAttachRootToWeapon))
      (Rendering|SetTranslucencySortDistanceOffset _output -2e+02))))

(event EventOnRemoval(EitherVFX-preferthisversion)
  (params (MyTarget :type Actor%20Object%20Reference) (Parameters :type Gameplay%20Cue%20Parameters%20Structure%20%28by%20ref%29) (ParticleComponents :type Array%20of%20FXSystem%20Component%20Object%20References) (AudioComponents :type Array%20of%20Audio%20Component%20Object%20References) (BurstCameraShakeInstance :type Camera%20Shake%20Base%20Object%20Reference) (BurstDecalInstance :type Decal%20Actor%20Object%20Reference))
  (bind (_returnvalue_bisactive _returnvalue_beamstartpoint _returnvalue_beamendpoint _returnvalue_hitnormal _returnvalue_bmadeimpact) (Utilities|Struct|BreakTargetedBeamData (Beam|GetBeamData (Variables|FortGameplayCueNotifyLoop_TargetedBeam|GetTargetedBeamActorComponent))))
  (bind (_returnvalue_bisactive_2 _returnvalue_beamstartpoint_2 _returnvalue_beamendpoint_2) (Utilities|Struct|BreakTargetedBeamData (Beam|GetBeamData (Variables|FortGameplayCueNotifyLoop_TargetedBeam|GetTargetedBeamActorComponent))))
  (Variables|Default|SetParticle%20Components_Removel ParticleComponents)
  (Utilities|Time|ClearandInvalidateTimerbyHandle 0 (Variables|AudioVisualizer|GetAVIconLoop_TimerHandle))
  (Utilities|FlowControl|Sequence
    (:then_0
      ; Only show removal impact FX if there was impact made
      (Rendering|SetVisibility (Utilities|Array|Get(acopy) (Variables|Default|GetParticle%20Components_Removel) 2) _returnvalue_bmadeimpact false)
      (Transformation|SetWorldLocation (Utilities|Array|Get(acopy) (Variables|Default|GetParticle%20Components_Removel) 2) (Transformation|GetWorldLocation (Variables|FortGameplayCueNotifyLoop_TargetedBeam|GetImpactFXComponent)) false false)
      (CallFunction|InitializeBeamAudio 0)
      (Rendering|SetVisibility (Variables|FortGameplayCueNotifyLoop_TargetedBeam|GetBeamComponents) false true)
      (Components|Activation|Deactivate (Variables|FortGameplayCueNotifyLoop_TargetedBeam|GetImpactFXComponent)))
    (:then_1
      (CallFunction|AttachSystemtoEyePositions MyTarget (Utilities|Array|MakeArray (Utilities|Array|Get(acopy) (Variables|Default|GetParticle%20Components_Removel))) (Utilities|Array|MakeArray (Utilities|Array|Get(acopy) (Variables|Default|GetParticle%20Components_Removel) 1)) (Variables|Default|GetbAttachRootToWeapon))
      (for _array_element (Variables|Default|GetParticle%20Components_Removel)
        (Effects|Components|ParticleSystem|SetBeamPoints _array_element 0 0 _returnvalue_beamstartpoint_2 _returnvalue_beamendpoint_2)))))

(event Custom|InitializeBeamAudio
  (params (FortPawnRef :type Actor%20Object%20Reference) (bPlay :type Boolean))
  ; Beam Audio Setup
  (if bPlay
    (bind _asfort_pawn (Utilities|Casting|CastToFortPawn FortPawnRef))
    (bind _output_get (Variables|Default|SetFortPawnRef _asfort_pawn))
    (Utilities|IsValid (Weapon|GetCurrentWeapon _output_get)
      (:Is%20Valid
        (bind _beamaudioactor (BeamAudio|GetBeamAudioActor(Message) (Weapon|GetCurrentWeapon _output_get)))
        (Utilities|IsValid _beamaudioactor
          (:Is%20Valid
            (bind _output_get_2 (Variables|Default|SetBeamAudioRef _beamaudioactor))
            (BeamAudio|PlayBeamAudio(Message) _output_get_2 true)
            ;;[1 0.2917 0.6822 0.5] Start Looping AV Icon
            (if (not (Pawn|IsLocallyViewed (Variables|Default|GetFortPawnRef)))
              (bind _returnvalue (Utilities|Time|SetTimerbyEvent (delegate Custom|ShowBeamAVIcon) (Variables|AudioVisualizer|GetAVIconLoopFrequency) true))
              (Variables|AudioVisualizer|SetAVIconLoop_TimerHandle _returnvalue))))))
    (else
      (bind _beamaudioref (Variables|Default|GetBeamAudioRef)
        (:then
          (BeamAudio|PlayBeamAudio(Message) _beamaudioref false))
        (:else)))))

(event Custom|ShowBeamAVIcon
  ; Beam Audio Setup
  ;;[1 0.2917 0.6822 0.5] Start Looping AV Icon
  (Fortnite|ShowSoundIndicatorLocal 0 0 (Variables|AudioVisualizer|GetBeamAudioEndPoint) 1.5e+04 "Gunshot"))

(event EventUpdateBeamAudio
  (params (StartPoint :type Vector :default "0, 0, 0") (EndPoint :type Vector :default "0, 0, 0") (bHitSuccess :type Boolean))
  ; Beam Audio Setup
  (Variables|AudioVisualizer|SetBeamAudioEndPoint EndPoint)
  (bind _beamaudioref (Variables|Default|GetBeamAudioRef)
    (:then
      (BeamAudio|UpdateBeamAudio(Message) _beamaudioref StartPoint EndPoint 0 true bHitSuccess))
    (:else)))

(fn AttachSystemToEyePositions
  (params (Actor :type Actor%20Object%20Reference) (SystemAttachToLeftEye :type Array%20of%20FXSystem%20Component%20Object%20References) (SystemAttachToRightEye :type Array%20of%20FXSystem%20Component%20Object%20References) (AttachToWeapon :type Boolean))
  (bind _asfort_player_pawn (Utilities|Casting|CastToFortPlayerPawn Actor)
    (:then
      (if AttachToWeapon
        (LocalVariables|SetWeaponSkeletonMesh (Variables|/Script/FortniteGame.FortWeapon|GetWeaponMesh (Variables|/Script/FortniteGame.FortPawn|GetCurrentWeapon _asfort_player_pawn)))
        (Utilities|FlowControl|Sequence
          (:then_0
            (for _array_element SystemAttachToLeftEye
              (Transformation|AttachComponentToComponent _array_element (LocalVariables|GetWeaponSkeletonMesh) "eye_l" "KeepRelative" "KeepRelative" "KeepRelative" true)))
          (:then_1
            (for _array_element_2 SystemAttachToRightEye
              (Transformation|AttachComponentToComponent _array_element_2 (LocalVariables|GetWeaponSkeletonMesh) "eye_r" "KeepRelative" "KeepRelative" "KeepRelative" true))))
        (else
          (LocalVariables|SetHeadSkeletonMesh (Customization|GetSkeletalMeshforPartType _asfort_player_pawn "Head" :!class /Script/FortniteGame.FortPlayerPawn))
          (Utilities|FlowControl|Sequence
            (:then_0
              (for _array_element_3 SystemAttachToLeftEye
                (Transformation|AttachComponentToComponent _array_element_3 (LocalVariables|GetHeadSkeletonMesh) "FACIAL_L_Eye" "KeepRelative" "KeepRelative" "KeepRelative" true)))
            (:then_1
              (for _array_element_4 SystemAttachToRightEye
                (Transformation|AttachComponentToComponent _array_element_4 (LocalVariables|GetHeadSkeletonMesh) "FACIAL_R_Eye" "KeepRelative" "KeepRelative" "KeepRelative" true)))))))
    (:CastFailed
      (Development|PrintString 0 "Cast Failed. Get Eye Position Failed" true true "(R=1.000000,G=0.000000,B=0.033119,A=1.000000)" 5.0))))

(fn LocateTargetedBeamComponent
  (params (Object :type Object%20Reference))
  (bind _foundcomponent (LocalVariables|GetFoundComponent))
  (bind _actorcomponentclass (Variables|Default|GetActorComponentClass))
  (bind _asfort_pawn (Utilities|Casting|CastToFortPawn Object)
    (:then
      (Utilities|IsValid (Actor|GetComponentbyClass _asfort_pawn _actorcomponentclass)
        (:Is%20Valid
          (LocalVariables|SetFoundComponent (Actor|GetComponentbyClass _asfort_pawn _actorcomponentclass)))
        (:Is%20Not%20Valid
          (LocalVariables|SetFoundComponent (Actor|GetComponentbyClass (Weapon|GetCurrentWeapon _asfort_pawn) _actorcomponentclass))))
      (Utilities|IsValid _foundcomponent
        (:Is%20Valid
          (bind _asfort_actor_component_targeted_beam (Utilities|Casting|CastToFortActorComponent_TargetedBeam _foundcomponent)
            (:then
              (Variables|FortGameplayCueNotifyLoop_TargetedBeam|SetTargetedBeamActorComponent _asfort_actor_component_targeted_beam))
            (:CastFailed
              (Development|PrintString 0 "Failed to cast found targeted beam component"))))))
    (:CastFailed
      (Development|PrintString 0 "Beam Owner is not a FortPawn"))))

(fn ConstructionScript)
