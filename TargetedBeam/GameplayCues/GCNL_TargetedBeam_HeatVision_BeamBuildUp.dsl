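; Looping gameplay cue that shows the heat vision targeted-beam build-up effect from a player's eyes.
; On start it attaches the first two particle systems to head (or optionally weapon) eye sockets and finds the TargetedBeam component on the pawn or current weapon.
; Tick interpolates the endpoint from beam data, with separate speeds for local and sim-proxy instigators, and updates particle beam points.
; An inactive beam hides the particles.
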(event EventOnLoopingStart(EitherVFX-preferthisversion)
  (params (MyTarget :type Actor%20Object%20Reference) (Parameters :type Gameplay%20Cue%20Parameters%20Structure%20%28by%20ref%29) (ParticleComponents :type Array%20of%20FXSystem%20Component%20Object%20References) (AudioComponents :type Array%20of%20Audio%20Component%20Object%20References))
  (bind (_returnvalue_bisactive _returnvalue_beamstartpoint _returnvalue_beamendpoint) (Utilities|Struct|BreakTargetedBeamData (Beam|GetBeamData (Variables|Components|GetTargetedBeamComponent))))
  (Utilities|FlowControl|Sequence
    (:then_0
      (CallFunction|AttachSystemtoEyePositions MyTarget (Utilities|Array|MakeArray (Utilities|Array|Get(acopy) ParticleComponents)) (Utilities|Array|MakeArray (Utilities|Array|Get(acopy) ParticleComponents 1))))
    (:then_1
      (Variables|Default|SetIsLocallyControlled (Ability|GameplayCue|IsInstigatorLocallyControlled Parameters))
      (CallFunction|LocateTargetedBeamComponent MyTarget)
      (Variables|Default|SetParticle%20Components ParticleComponents)
      (Variables|Default|SetCurrentEndpoint _returnvalue_beamendpoint))))

(event EventTick
  (params (DeltaSeconds :type Float%20%28single-precision%29))
  (bind _targetedbeamcomponent (Variables|Components|GetTargetedBeamComponent))
  (bind (_returnvalue_bisactive _returnvalue_beamstartpoint _returnvalue_beamendpoint) (Utilities|Struct|BreakTargetedBeamData (Beam|GetBeamData _targetedbeamcomponent)))
  (if _returnvalue_bisactive
    (Variables|Default|SetCurrentEndpoint (Math|Interpolation|VInterpTo (Variables|Default|GetCurrentEndpoint) _returnvalue_beamendpoint DeltaSeconds (Math|Float|SelectFloat (ScalableFloat|GetValueatLevel (Variables|Default|GetBeamInterpSpeed_Local)) (ScalableFloat|GetValueatLevel (Variables|Default|GetBeamInterpSpeed_SimProxy)) (Variables|Default|GetIsLocallyControlled))))
    (for _array_element (Variables|Default|GetParticle%20Components)
      (Effects|Components|ParticleSystem|SetBeamPoints _array_element 0 0 _returnvalue_beamstartpoint (Variables|Default|GetCurrentEndpoint)))
    (else
      (Rendering|SetVisibility (Variables|Default|GetParticle%20Components) false true))))

(fn AttachSystemToEyePositions
  (params (Actor :type Actor%20Object%20Reference) (SystemAttachToLeftEye :type Array%20of%20FXSystem%20Component%20Object%20References) (SystemAttachToRightEye :type Array%20of%20FXSystem%20Component%20Object%20References) (AttachToWeapon :type Boolean))
  (bind _asfort_player_pawn (Utilities|Casting|CastToFortPlayerPawn Actor)
    (:then
      (if AttachToWeapon
        (LocalVariables|SetWeaponSkeletonMesh (Variables|/Script/FortniteGame.FortWeapon|GetWeaponMesh (Variables|/Script/FortniteGame.FortPawn|GetCurrentWeapon _asfort_player_pawn)))
        (Utilities|FlowControl|Sequence
          (:then_0
            (for _array_element SystemAttachToLeftEye
              (Transformation|AttachComponentToComponent _array_element (LocalVariables|GetWeaponSkeletonMesh) "eye_l" "KeepRelative" "KeepRelative" "KeepRelative" true)))
          (:then_1
            (for _array_element_2 SystemAttachToRightEye
              (Transformation|AttachComponentToComponent _array_element_2 (LocalVariables|GetWeaponSkeletonMesh) "eye_r" "KeepRelative" "KeepRelative" "KeepRelative" true))))
        (else
          (LocalVariables|SetHeadSkeletonMesh (Customization|GetSkeletalMeshforPartType _asfort_player_pawn "Head" :!class /Script/FortniteGame.FortPlayerPawn))
          (Utilities|FlowControl|Sequence
            (:then_0
              (for _array_element_3 SystemAttachToLeftEye
                (Transformation|AttachComponentToComponent _array_element_3 (LocalVariables|GetHeadSkeletonMesh) "FACIAL_L_Eye" "KeepRelative" "KeepRelative" "KeepRelative" true)))
            (:then_1
              (for _array_element_4 SystemAttachToRightEye
                (Transformation|AttachComponentToComponent _array_element_4 (LocalVariables|GetHeadSkeletonMesh) "FACIAL_R_Eye" "KeepRelative" "KeepRelative" "KeepRelative" true)))))))
    (:CastFailed
      (Development|PrintString 0 "Cast Failed. Get Eye Position Failed" true true "(R=1.000000,G=0.000000,B=0.033119,A=1.000000)" 5.0))))

(fn LocateTargetedBeamComponent
  (params (Object :type Actor%20Object%20Reference))
  (bind _foundcomponent (LocalVariables|GetFoundComponent))
  (bind _actorcomponentclass (Variables|Default|GetActorComponentClass))
  (bind _asfort_pawn (Utilities|Casting|CastToFortPawn Object)
    (:then
      (Utilities|IsValid (Actor|GetComponentbyClass _asfort_pawn _actorcomponentclass)
        (:Is%20Valid
          (LocalVariables|SetFoundComponent (Actor|GetComponentbyClass _asfort_pawn _actorcomponentclass)))
        (:Is%20Not%20Valid
          (LocalVariables|SetFoundComponent (Actor|GetComponentbyClass (Weapon|GetCurrentWeapon _asfort_pawn) _actorcomponentclass))))
      (Utilities|IsValid _foundcomponent
        (:Is%20Valid
          (bind _asfort_actor_component_targeted_beam (Utilities|Casting|CastToFortActorComponent_TargetedBeam _foundcomponent)
            (:then
              (Variables|Components|SetTargetedBeamComponent _asfort_actor_component_targeted_beam))
            (:CastFailed
              (Development|PrintString 0 "Failed to cast found targeted beam component"))))))
    (:CastFailed
      (Development|PrintString 0 "Beam Owner is not a FortPawn"))))

(fn ConstructionScript)
